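# Sound cue for the looping thrust sound of the Booty Buoy cosmetic glider.
# A single looping wave player feeds the root directly with the thrust loop wave, with no randomization or mixing nodes.
# Playback is shaped by the Athena player glider thrust attenuation, a cosmetic glider loop audio properties sheet and the glider sound class.
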SoundCue /BRCosmetics/Sounds/Gliders/BootyBuoy/Glider_Booty_Buoy_Thrust_Loop_Cue
root=WavePlayer_1
nodes [{id=WavePlayer_1, properties {bLooping=true, soundWaveAssetPtr=&/Glider_Booty_Buoy_Thrust_Loop}}]
properties {
 attenuationSettings=&Fort_Player_Glider_Thrust_Athena_Att
 audioPropertiesSheet=&APS_Cosmetic_Glider_Loop
 maxDistance=12500
 soundClassObject=&Fort_Gliders_Land
 volumeMultiplier=1
}
